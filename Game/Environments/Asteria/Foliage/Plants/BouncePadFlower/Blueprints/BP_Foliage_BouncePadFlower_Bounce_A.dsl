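; Bounce pad flower foliage that launches overlapping pawns upward and slightly forward, with separate handling for hoverboards and physics objects, which destroy it.
; An enum sets launch height at BeginPlay, and a capsule-trace ceiling check caps it under obstructions.
; Pawn bounces zero velocity, apply launched and fall-damage-immunity effects, send a gameplay event and trigger a burst cue.
; Remote machines play a timeline squash animation; HandleHeal and cooldown events are empty here.
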(event EventBeginPlay
  (switch Utilities|FlowControl|Switch|SwitchonDaisyBouceHeightsEnum (Variables|Default|GetBouceHeights)
    (:NewEnumerator0
      (Variables|Default|SetDaisyBouceHeight (ScalableFloat|GetValueatLevel (Variables|Default|GetBounceHeight_2Floors))))
    (:NewEnumerator1
      (Variables|Default|SetDaisyBouceHeight (ScalableFloat|GetValueatLevel (Variables|Default|GetBounceHeight_4Floors))))))

(event OnComponentBeginOverlap(Jump_Collision)
  (params (OverlappedComponent :type Primitive%20Component%20Object%20Reference) (OtherActor :type Actor%20Object%20Reference) (OtherComp :type Primitive%20Component%20Object%20Reference) (OtherBodyIndex :type Integer) (bFromSweep :type Boolean) (SweepResult :type Hit%20Result%20Structure%20%28by%20ref%29))
  (bind _returnvalue (Ability|GetAbilitySystemComponent (Variables|Default|GetOverlapPawn)))
  (bind (_returnvalue_x _returnvalue_y) (Math|Vector|BreakVector (* (Transformation|GetActorForwardVector (Variables|Default|GetOverlapPawn)) (Variables|Default|GetForwardVectorAddition))))
  (bind _healtarget (Variables|Default|GetHealTarget))
  (bind _overlapcontroller (Variables|Default|GetOverlapController))
  (bind _overlappawn (Variables|Default|GetOverlapPawn))
  (Networking|FlushNetDormancy)
  (bind _asfort_pawn (Utilities|Casting|CastToFortPawn OtherActor)
    (:then
      (bind _output_get (Variables|Default|SetOverlapPawn _asfort_pawn))
      ;[1 1 1] Do nothing if we are attached to a AIPawn (Riding Creatures)
      (Utilities|Casting|CastToFortAIPawn (Actor|GetAttachParentActor _output_get)
        (:CastFailed
          (if (not (or (GameplayTags|HasMatchingGameplayTag (Ability|GetAbilitySystemComponent _output_get) (Variables|Default|GetPlayerLaunched)) (Variables|/Script/FortniteGame.FortPawn|GetbIsDBNO _output_get)))
            (bind _asfort_player_controller (Utilities|Casting|CastToFortPlayerController (Pawn|GetController _output_get))
              (:then
                (Variables|Default|SetOverlapController _asfort_player_controller))
              (:CastFailed
                _))
            ;[1 1 1] Ceiling check
            (bind _success (CallFunction|CeilingCheck))
            (if _success
              (Variables|Default|SetBounceHeight 6e+02)
              (else
                (Variables|Default|SetBounceHeight (Variables|Default|GetDaisyBouceHeight))))
            (Variables|Default|SetFinalBounceJumpHeight (Variables|Default|GetBounceHeight))
            (Variables|/Script/Engine.MovementComponent|SetVelocity "0, 0, 0" (Variables|/Script/Engine.Character|GetCharacterMovement _overlappawn)) ; so that launch distance is consistent
            (Character|LaunchCharacter _overlappawn (Math|Vector|MakeVector _returnvalue_x _returnvalue_y (Variables|Default|GetFinalBounceJumpHeight)) false true)
            (Variables|Default|SetHealTarget _overlappawn)
            (for _array_element (Actor|GetAttachedActors _overlappawn true false)
              (bind _asfort_player_pawn (Utilities|Casting|CastToFortPlayerPawn _array_element))
              (Variables|Default|SetHealTarget _asfort_player_pawn))
            ;[0.6958 0 1] Apply fall damage immunity and heal GEs. Play impact FX.
            (GameplayEffects|ApplyGameplayEffectToSelf :self _returnvalue :GameplayEffectClass (Variables|Default|GetGE_PlayerLaunched) :Level 0.0)
            (GameplayEffects|ApplyGameplayEffectToSelf :self _returnvalue :GameplayEffectClass (Variables|Default|GetGE_FallDamageImmunity) :Level 0.0)
            ;[1 1 1] Bounced Gameplay Event
            (Ability|SendGameplayEventtoActor (Fortnite|GetFortPlayerControllerfromActor (Variables|Default|GetOverlapPawn)) (Variables|Default|GetEvent%20Tag) (Utilities|Struct|MakeGameplayEventData (Variables|Default|GetEvent%20Tag) self (Variables|Default|GetOverlapPawn)))
            (GameplayCue|ExecuteGameplayCueOnActor(Burst) self (Variables|Default|GetGC_TriggeredBounce) (Ability|GameplayCue|MakeGameplayCueParameters :NormalizedMagnitude (Math|Conversions|ToFloat(Boolean) (Variables|Default|GetHealOnCooldown)) :RawMagnitude 0.0 :Location "0,0,0" :Normal "0,0,0" :Instigator _overlappawn :GameplayEffectLevel 1 :AbilityLevel 1 :bReplicateLocationWhenUsingMinimalRepProxy false))
            (Utilities|Casting|CastToFortPlayerPawn _healtarget
              (:then
                (CallFunction|HandleHeal (Ability|GetAbilitySystemComponent _healtarget)))
              (:CastFailed
                _))
            (Variables|/Script/FortniteGame.FortPawn|SetbIgnoreNextFallingDamage true _overlappawn)
            (Utilities|FlowControl|Sequence
              (:then_0
                (Networking|SwitchHasAuthority
                  (:Authority
                    (CallFunction|BounceAnim))))
              (:then_1
                (Utilities|IsValid _overlapcontroller
                  (:Is%20Valid
                    (Game|Feedback|LocalPlayForceFeedback _overlapcontroller (Variables|Default|GetLaunchForceFeedback) false "TireLaunch"))
                  (:Is%20Not%20Valid
                    _))
                (Variables|Default|SetOverlapPawn 0)
                (Variables|Default|SetOverlapController 0)
                (Variables|Default|SetHealTarget 0)))))))
    (:CastFailed
      (bind _returnvalue_2 (FortPhysicsSimulationLibrary|IsPhysicsObjectActiveinSimulation OtherActor))
      (if _returnvalue_2
        (Actor|DestroyActor)
        (else
          ;[1 0.01562 0.01562] Legacy Vehicles?
          ;;[1 1 1] Ceiling check
          (bind _success_2 (CallFunction|CeilingCheck))
          (if _success_2
            (Variables|Default|SetBounceHeight 600)
            (else
              (Variables|Default|SetBounceHeight 2e+03)))
          (Variables|Default|SetFinalBounceJumpHeight (Variables|Default|GetBounceHeight))
          (Vehicles|LaunchHoverboard OtherActor (Math|Vector|MakeVector 0.0 0.0 (Variables|Default|GetFinalBounceJumpHeight)) "TireLaunch"))))))

(event Custom|HandleHeal
  (params (AbilitySystem :type Ability%20System%20Component%20Object%20Reference)))

(event Custom|BounceAnim
  (Networking|SwitchHasAuthority
    (:Remote
      (bind (_direction _emissive_mult _scale_xy _scale_z) (Timeline|Bounce%20Emissive PlayFromStart)
        (:Update
          (Transformation|SetWorldScale3D (Variables|Debug|GetStaticMeshComponent) (* (Math|Vector|MakeVector _scale_xy _scale_xy _scale_z) (Variables|Default|GetOrigScale))))))))

(event Custom|ResetHealCooldown)

(event Custom|ToggleAmbientFX
  (params (Active :type Boolean)))

(event Custom|ResetMinMaxScales
  (bind (_x _y) (Math|Vector2D|BreakVector2D (Variables|GetRandomScaleRange)))
  ;[1 1 1] Allows designers to reset the min max scale of this actor if it's changed
  (Transformation|SetActorScale3D self (Math|Vector|MakeVector (Math|Random|RandomFloatinRange _x _y) (Math|Random|RandomFloatinRange _x _y) (Math|Random|RandomFloatinRange _x _y))))

(fn CeilingCheck
  (bind _overlappawn (Variables|Default|GetOverlapPawn)
    (:then
      (Utilities|Array|Add (LocalVariables|GetActorsToIgnore) _overlappawn)
      (Utilities|Array|Add (LocalVariables|GetActorsToIgnore) (Actor|GetAttachParentActor _overlappawn))
      (Utilities|Array|AppendArray (LocalVariables|GetActorsToIgnore) (Actor|GetAttachedActors _overlappawn true false))
      (for _array_element (Actor|GetAttachedActors _overlappawn true false)
        (Utilities|Array|AppendArray (LocalVariables|GetActorsToIgnore) (Actor|GetAttachedActors _array_element true false))))
    (:else
      _))
  (bind (_outhit _returnvalue) (Collision|CapsuleTraceByChannel (Math|Vector|MakeVector (.x (Transformation|GetWorldLocation (Variables|Transformation|GetRootComponent))) (.y (Transformation|GetWorldLocation (Variables|Transformation|GetRootComponent))) (+ (.z (Transformation|GetWorldLocation (Variables|Transformation|GetRootComponent))) 200)) (Math|Vector|MakeVector (.x (Transformation|GetWorldLocation (Variables|Transformation|GetRootComponent))) (.y (Transformation|GetWorldLocation (Variables|Transformation|GetRootComponent))) (+ (.z (Transformation|GetWorldLocation (Variables|Transformation|GetRootComponent))) 340)) 50 100 "TraceTypeQuery9" false (LocalVariables|GetActorsToIgnore)))
  (return _returnvalue))

(fn GetAffiliationsToShowFor
  (return (Utilities|Array|MakeArray "Neutral" "Hostile")))

(fn ShouldShowSoundIndicator
  (params (PlayerController :type Fort%20Player%20Controller%20Object%20Reference))
  (bind _healoncooldown (Variables|Default|GetHealOnCooldown))
  (bind _returnvalue (not _healoncooldown))
  (bind _healtarget (Variables|Default|GetHealTarget))
  (Utilities|IsValid _healtarget
    (:Is%20Valid
      (return (and (not (== (Fortnite|GetFortPlayerControllerfromActor _healtarget) PlayerController)) _returnvalue)))
    (:Is%20Not%20Valid
      (return (and _returnvalue (not (== PlayerController (Fortnite|GetFortPlayerControllerfromActor (Variables|Default|GetOverlapPawn)))))))))

(fn ConstructionScript
  (Variables|Default|SetOrigScale (Transformation|GetWorldScale (Variables|Debug|GetStaticMeshComponent))))
