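# Parent material for the Asteria bounce pad flower plant, Nanite-enabled, layering an optional raindrop pass and center-specific shading over two-sided plant attributes.
# Glow combines a screen-aligned sphere mask anchored at a local point, noisy Fresnel and animated stars, with time-of-day brightness.
# World position offset logic covers wind bending through an animation lookup texture, physics-field and player-proximity reactions, pulsing, and ScaleMult-driven bounce deformation.
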Material /Game/Environments/Asteria/Foliage/Plants/BouncePadFlower/Materials/M_BouncePadFlower_Parent
graph {
 properties {bUseMaterialAttributes=true, bUsedWithNanite=true, pixelDepthOffsetMode=PDOM_Legacy, usedShadingModels=Default Lit}
 nodes [
  n1 = MaterialFunctionCall() {materialFunction=&DebugTimeSine}
  Add(A=@n1.Result)
  n2 = Multiply(A=NamedRerouteUsage(), B=ScalarParameter() {defaultValue=0.5025, group=AnimationTweaking, parameterName=Leaf_MaxBend_0to1})
  n3 = LinearInterpolate(
   A=@n2
   B=Multiply(A=NamedRerouteUsage(), B=ScalarParameter() {defaultValue=1, group=Animation Physics, parameterName=Leaf_PhysicsField_MaxBend})
   Alpha=NamedRerouteUsage()
  )
  n4 = NamedRerouteUsage()
  n5 = ComponentMask(None=TextureCoordinate() {coordinateIndex=1}) {r=true}
  n6 = OneMinus(None=@n5)
  n7 = TextureSample(UVs=AppendVector(A=StaticSwitch(True=@n3, False=@n2, Value=@n4), B=Saturate(None=@n6))) {
   samplerSource=SSM_Clamp_WorldGroupSettings
   samplerType=SAMPLERTYPE_LinearColor
   texture=/Game/Environments/Asteria/Foliage/Grass/MasterMaterial/GenericShared_Grass_Foliage_Anim
  }
  n8 = NamedRerouteUsage()
  n9 = StaticSwitch(True=LinearInterpolate(A=@n8, B=NamedRerouteUsage(), Alpha=NamedRerouteUsage()), False=@n8, Value=NamedRerouteUsage())
  n10 = Multiply(A=@n7.G, B=ScalarParameter() {defaultValue=1, group=AnimationTweaking, parameterName=Leaf_Bend_Boost})
  n11 = VertexColor()
  n12 = Transform(None=Subtract(A=Multiply(A=@n11) {constB=2})) {transformSourceType=TRANSFORMSOURCE_Instance}
  n13 = ScalarParameter() {defaultValue=1, group=AnimationTweaking, parameterName=Leaf_Bend_falloff}
  n14 = Multiply(
   A=Add(A=Multiply(A=@n7.R, B=@n9), B=Multiply(A=@n10, B=@n12))
   B=Multiply(A=Power(Base=@n5, Exp=@n13), B=ScalarParameter() {defaultValue=1, group=AnimationTweaking, parameterName=Leaf_Bend_Boost_XY})
  )
  Add(A=NamedRerouteUsage(), B=@n14)
  Constant() {r=0.5}
  n15 = NamedRerouteUsage()
  n16 = StaticSwitch(True=LinearInterpolate(A=@n15, B=NamedRerouteUsage(), Alpha=NamedRerouteUsage()), False=@n15, Value=NamedRerouteUsage())
  LinearInterpolate(A=@n16, B=NamedRerouteUsage(), Alpha=NamedRerouteUsage())
  n17 = Multiply(A=NamedRerouteUsage(), B=ScalarParameter() {defaultValue=1, group=AnimationTweaking, parameterName=Overall_MaxBend_0to1})
  n18 = NamedRerouteUsage()
  n19 = ScalarParameter() {defaultValue=1, group=Animation Physics, parameterName=Overall_PhysicsField_MaxBend}
  n20 = StaticSwitch(True=LinearInterpolate(A=@n17, B=Multiply(A=@n18, B=@n19), Alpha=NamedRerouteUsage()), False=@n17, Value=NamedRerouteUsage())
  LinearInterpolate(A=@n20, B=ScalarParameter() {defaultValue=1, group=Animation, parameterName=Overall_PlayerBend}, Alpha=NamedRerouteUsage())
  n21 = Multiply(A=NamedRerouteUsage(), B=VectorParameter() {parameterName=SSC, defaultValue=<rgba 0.354 0.1432 0.0867 1>}.RGB) {constB=0.5}
  LinearInterpolate(A=NamedRerouteUsage(), B=@n21, Alpha=NamedRerouteUsage())
  LinearInterpolate(Alpha=@n1.Result) {constA=0.7, constB=1.25}
  n22 = TextureSampleParameter2D() {group=Tex, parameterName=Diffuse, texture=/Game/Environments/Apollo/Foliage/Nightmare/Textures/T_Nightmare_JunglePlants_01_D}
  n23 = TextureSampleParameter2D() {
   parameterName=SpecularMasks
   samplerType=SAMPLERTYPE_Masks
   texture=/Game/Environments/Apollo/Foliage/Nightmare/Textures/T_Nightmare_JunglePlants_01_S
  }
  n24 = ScalarParameter() {defaultValue=0.5, group=S, parameterName=S Inside}
  n25 = ScalarParameter() {defaultValue=0.25, group=S, parameterName=S Out}
  n26 = TwoSidedSign()
  n27 = ScalarParameter() {defaultValue=1, group=R, parameterName=R Inside}
  n28 = ScalarParameter() {defaultValue=0.65, group=R, parameterName=R Out}
  n29 = VectorParameter() {group=SSS, parameterName=SSS Tone, channelNames {a=SSS Value, b="", g="", r=""}, defaultValue=<rgba 2 1.438 0.4285 0.5>}
  n30 = TextureSampleParameter2D() {
   parameterName=Normals
   samplerType=SAMPLERTYPE_Normal
   texture=/Game/Environments/Apollo/Foliage/Nightmare/Textures/T_Nightmare_JunglePlants_01_N
  }
  n31 = MakeMaterialAttributes!(
   BaseColor=@n22.RGB
   Metallic
   Specular=Multiply(A=@n23.R, B=LinearInterpolate(A=@n24, B=@n25, Alpha=@n26))
   Roughness=Multiply(A=@n23.B, B=LinearInterpolate(A=@n27, B=@n28, Alpha=@n26))
   Anisotropy
   EmissiveColor
   Opacity=LinearInterpolate(B=@n29.SSS Value, Alpha=@n26)
   OpacityMask=@n22.RGB
   Normal=Multiply(A=@n26, B=@n30.RGB)
   Tangent
   WorldPositionOffset
   SubsurfaceColor=Multiply(A=@n22.RGB, B=@n29)
   ClearCoat
   ClearCoatRoughness
   AmbientOcclusion
   Refraction
   CustomizedUVs_0
   CustomizedUVs_1
   CustomizedUVs_2
   CustomizedUVs_3
   CustomizedUVs_4
   CustomizedUVs_5
   CustomizedUVs_6
   CustomizedUVs_7
   PixelDepthOffset
   ShadingModel
   Displacement
  )
  n32 = GetMaterialAttributes(None=NamedRerouteDeclaration(None=@n31) {name=Just The Plant, nodeColor=<rgba 1 0 0.7266 1>}) {attributeGetTypes [
    #1
    #2
    #3
    #4
   ]}
  n33 = Multiply(A=TextureCoordinate(), B=ScalarParameter() {defaultValue=4, parameterName=Scale Drops})
  n34 = TextureSample(UVs=@n33) {samplerType=SAMPLERTYPE_Normal, texture=/Game/Global/Base4Layers_2022/Textures/T_RainDrop_N}
  n35 = DotProduct(A=AtmosphericLightVector(), B=Transform(None=@n34.RGB))
  n36 = TextureSample(UVs=@n33) {samplerType=SAMPLERTYPE_Masks, texture=/Game/Global/Base4Layers_2022/Textures/T_RainDrop_S}
  n37 = LinearInterpolate(B=LinearInterpolate(A=Constant() {r=2}, B=Constant() {r=0.35}, Alpha=Saturate(None=@n35)), Alpha=@n36.G) {constA=1}
  n38 = Multiply(A=@n34.RGB, B=Constant3Vector() {constant=<rgba 1 1 0 0>})
  n39 = SetMaterialAttributes!(
   MaterialAttributes=@n32.MaterialAttributes
   Base Color=Multiply(A=@n32.Base Color, B=@n37)
   Specular=Multiply(A=@n32.Specular, B=@n36.B)
   Roughness=Multiply(A=@n32.Roughness, B=@n36.B)
   Normal=LinearInterpolate(A=Add(A=@n32.Normal, B=@n38), B=@n34.RGB)
  ) {attributeSetTypes [
    #1
    #2
    #3
    #4
   ]}
  n40 = Saturate(None=TwoSidedSign())
  n41 = GetMaterialAttributes(None=BlendMaterialAttributes(A=@n32.MaterialAttributes, B=@n39, Alpha=@n40) {vertexAttributeBlendType=UseA}) {attributeGetTypes [#5]}
  n42 = SetMaterialAttributes!(MaterialAttributes=@n41.MaterialAttributes, Subsurface Color=Multiply(A=@n41.Subsurface Color, B=Max(A=@n37) {constB=0.5})) {attributeSetTypes [#5]}
  n43 = GetMaterialAttributes(None=StaticSwitchParameter(True=BlendMaterialAttributes(A=@n42, B=@n41.MaterialAttributes, Alpha=@n40) {vertexAttributeBlendType=UseA}, False=NamedRerouteUsage()) {defaultValue=true, parameterName=Raindrops}) {attributeGetTypes [
    #6
    #1
    #7
    #2
    #3
   ]}
  n44 = ScalarParameter() {defaultValue=0.3, parameterName=Diffuse_Min}
  n45 = Multiply(A=@n44) {constB=3}
  n46 = LinearInterpolate(
   A=MaterialFunctionCall(Value_Morning=@n44, Value_Day=@n44, Value_Evening=@n45, Value_Night=@n45) {materialFunction=&Scalar_Time_of_Day}.Result
   B=ScalarParameter() {defaultValue=1, parameterName=Diffuse_Max}
   Alpha=NamedRerouteUsage()
  ) {constA=0.2, constB=0.8}
  n47 = NamedRerouteUsage()
  n48 = SetMaterialAttributes!(
   MaterialAttributes=@n43.MaterialAttributes
   World Position Offset=NamedRerouteUsage()
   Emissive Color=NamedRerouteUsage()
   Base Color=LinearInterpolate(A=@n43.Base Color, B=Multiply(A=@n46, B=@n43.Base Color) {constB=0.8}, Alpha=NamedRerouteUsage())
   Roughness=LinearInterpolate(A=@n43.Roughness, B=ScalarParameter() {defaultValue=0.4, parameterName=Roughness_Center}, Alpha=@n47)
   Specular=LinearInterpolate(A=@n43.Specular, B=ScalarParameter() {defaultValue=0.2, parameterName=Specular_Center}, Alpha=@n47)
   Subsurface Color=NamedRerouteUsage()
   Opacity=NamedRerouteUsage()
  ) {attributeSetTypes [
    #6
    #8
    #1
    #3
    #2
    #5
    #9
   ]}
  MaterialFunctionCall(MA In=@n48) {materialFunction=&MF_CreativeMoveTool}
  MaterialFunctionCall!(
   Previous WPO Value
   Force Magnitude Mask
   Force Magnitude Multiplier
   Project Field Force Vector to Normals
   Noise Size
   Noise Pan Speed
   Noise Falloff
   Stretch Noise Coordinates Z
  ) {desc=Sample global physics field add to WPO, materialFunction=/Game/Effects/Niagara/MasterMaterials/Functions/MF_FX_PhysicsFields_Grass}
  n49 = MaterialFunctionCall() {materialFunction=&BoundingBoxBased_0-1_UVW}
  n50 = OneMinus(None=Abs(None=ConstantBiasScale(None=@n49.RGB) {bias=-0.5, scale=2}))
  n51 = ConstantBiasScale(None=@n49.G) {bias=-0.5, scale=2}
  MaterialFunctionCall(X=@n50, Y=OneMinus(None=Abs(None=@n51)), Z=Constant() {r=1}) {materialFunction=&MakeFloat3}
  n52 = MaterialFunctionCall() {materialFunction=&ObjectScale}
  NamedRerouteDeclaration(None=LinearInterpolate(A=Multiply(A=@n14, B=@n52.Scale Z), Alpha=NamedRerouteUsage()) {constB=0}) {name=WindAnim, nodeColor=<rgba 1 0 0.2109 1>}
  n53 = OneMinus(None=VectorParameter() {parameterName=ScaleMult, defaultValue=<rgba 1 1 1 1>}.RGB)
  n54 = Subtract(A=TransformPosition(None=Constant3Vector()) {transformType=TRANSFORMPOSSOURCE_World}, B=WorldPosition().XYZ)
  NamedRerouteDeclaration(None=Multiply(A=Multiply(A=@n53, B=@n54), B=@n50)) {name=BounceWPO, nodeColor=<rgba 0 0.6719 1 1>}
  n55 = TextureSample() {samplerType=SAMPLERTYPE_Masks, texture=/Game/Environments/Asteria/Foliage/Plants/BouncePadFlower/Textures/T_Foliage_BouncePadFlower_A_A_M}
  n56 = Power(Base=@n55.G) {constExponent=10}
  n57 = VectorParameter() {parameterName=EmissiveColor, defaultValue=<rgba 1 0.3177 0 1>}
  n58 = ScalarParameter() {parameterName=EmissiveMult}
  n59 = Multiply(A=@n58) {constB=0.2}
  n60 = MaterialFunctionCall(Value_Morning=@n59, Value_Day=@n58, Value_Evening=@n59, Value_Night=@n59) {materialFunction=&Scalar_Time_of_Day}
  NamedRerouteDeclaration(None=Multiply(A=@n56, B=Multiply(A=@n57.RGB, B=@n60.Result))) {name=CenterEmissive, nodeColor=<rgba 0.7109 0 1 1>}
  n61 = Multiply(
   A=ComponentMask(None=TextureCoordinate() {coordinateIndex=1}) {g=true}
   B=ScalarParameter() {defaultValue=1308.7753, group=AnimationTweaking, parameterName=Leaf_RandomFrequency}
  )
  n62 = ScalarParameter() {defaultValue=246.1449, group=AnimationTweaking, parameterName=Leaf_DragWiggle_Calm}
  n63 = ScalarParameter() {defaultValue=120, group=AnimationTweaking, parameterName=Leaf_DragWiggle_Storm}
  n64 = NamedRerouteUsage()
  n65 = Add(A=@n61, B=Multiply(A=VertexColor().A, B=LinearInterpolate(A=@n62, B=@n63, Alpha=@n64)))
  n66 = AppendVector(A=@n65, B=@n65)
  n67 = Constant3Vector()
  n68 = Multiply(
   A=NamedRerouteDeclaration(None=TransformPosition(None=@n67) {transformSourceType=TRANSFORMPOSSOURCE_Instance, transformType=TRANSFORMPOSSOURCE_World}) {name=BasePivot, nodeColor=<rgba 1 0 0.7031 1>}
   B=ScalarParameter() {defaultValue=1408.6052, group=AnimationTweaking, parameterName=Overall_Bend_Frequency}
  )
  n69 = StaticBool() {value=true}
  n70 = MaterialFunctionCall!(
   Position=Add(A=AppendVector(A=@n66, B=@n61), B=@n68)
   IsTrees
   IsGrass
   IsMediumPlant=@n69
   IsMediumPlant_2024Anim
   IsMediumPlant2ndLevel
   MediumPlant_BasePivotPostion
  ) {materialFunction=/Game/Environments/Asteria/Foliage/Wind/MF_WindActorData}
  NamedRerouteDeclaration(None=@n70.WindIntensitySingleSample) {name=WindIntensityNoise, nodeColor=<rgba 0.1719 0 1 1>}
  NamedRerouteDeclaration(None=@n70.WindI_Direction_WithNoise) {name=WindDirectionWithNoise, nodeColor=<rgba 0 1 0.9297 1>}
  n71 = MaterialFunctionCall(Position=@n68, IsMediumPlant=@n69, IsMediumPlant2ndLevel=StaticBool() {value=true}, MediumPlant_BasePivotPostion=NamedRerouteUsage()) {materialFunction=&MF_WindActorData}
  NamedRerouteDeclaration(None=@n71.WindI_Direction_WithNoise) {name=WindDirectionOverallBend, nodeColor=<rgba 0 0.1562 1 1>}
  NamedRerouteDeclaration(None=@n71.WindIntesnity) {name=WindIntensityNoise_Overall, nodeColor=<rgba 1 0 0.4219 1>}
  NamedRerouteDeclaration(None=StaticBool() {value=true}) {name=Use Physics, nodeColor=<rgba 0 1 0.3438 1>}
  NamedRerouteDeclaration(None=@n70.StormMask) {name=StormMask, nodeColor=<rgba 0.3125 0 1 1>}
  n72 = CollectionParameter() {collection=&FortniteMaterialParameters, parameterName=PlayerPosition}
  n73 = ComponentMask(None=@n72) {b=true, g=true, r=true}
  n74 = NamedRerouteUsage()
  n75 = Saturate(None=SphereMask(
    A=@n73
    B=@n74
    Radius=ScalarParameter() {defaultValue=200, group=Animation Physics, parameterName=MaxDistanceForPlayerInteraction}
    Hardness=Constant()
   ) {attenuationRadius=64, hardnessPercent=20}) {desc=PlayerSphereMask}
  NamedRerouteDeclaration(None=@n75) {name=PlayerSphereMask, nodeColor=<rgba 0 0.8828 1 1>}
  NamedRerouteDeclaration(None=Normalize(VectorInput=Subtract(A=@n74, B=@n73))) {name=PlayerPushDirection, nodeColor=<rgba 0 0.6016 1 1>}
  n76 = Multiply(
   A=Subtract(
    A=WorldPosition().XYZ
    B=MaterialFunctionCall() {materialFunction=&ObjectPivotPoint}.Object Pivot Location
   )
   B=MaterialFunctionCall(Speed=ScalarParameter() {defaultValue=0.5, parameterName=PulsingSpeed}) {materialFunction=&DebugTimeSine}.Result
  )
  n77 = ScalarParameter() {defaultValue=0.015, parameterName=WPOPuslingScale}
  NamedRerouteDeclaration(None=LinearInterpolate(B=Multiply(A=@n76, B=@n77), Alpha=NamedRerouteUsage())) {name=PulsingWPO, nodeColor=<rgba 1 0 0.375 1>}
  n78 = VertexColor()
  n79 = TextureSample(UVs=AppendVector(A=@n20, B=OneMinus(None=@n78.A))) {
   samplerSource=SSM_Clamp_WorldGroupSettings
   samplerType=SAMPLERTYPE_LinearColor
   texture=&GenericShared_Grass_Foliage_Anim
  }
  n80 = ScalarParameter() {defaultValue=1.3, group=AnimationTweaking, parameterName=Overall_Bend_XY}
  n81 = Multiply(A=@n79.G) {constB=-1}
  n82 = ScalarParameter() {defaultValue=1.3, group=AnimationTweaking, parameterName=Overall_Bend_Z}
  n83 = Add(A=Multiply(A=Multiply(A=@n79.R, B=@n80), B=@n16), B=AppendVector(A=Constant2Vector(), B=Multiply(A=@n81, B=@n82)))
  NamedRerouteDeclaration(None=@n83) {name=Cheap OverallBend, nodeColor=<rgba 0.5 1 0 1>}
  NamedRerouteDeclaration(None=@n21) {name=SSColor, nodeColor=<rgba 0 1 0.3203 1>}
  n84 = NamedRerouteUsage()
  NamedRerouteDeclaration(None=LinearInterpolate(
    A=ScalarParameter() {defaultValue=0.5, parameterName=Opacity_Petal_Max}
    B=ScalarParameter() {defaultValue=0.2, parameterName=Opacity_Petal_Min}
    Alpha=Power(Base=@n84) {constExponent=0.5}
   ) {constB=0.25}) {name=Opacity, nodeColor=<rgba 1 0.5156 0 1>}
  n85 = NamedRerouteUsage()
  n86 = LinearInterpolate(B=NamedRerouteUsage(), Alpha=Saturate(None=Add(A=@n56))) {constB=0}
  NamedRerouteDeclaration(None=Add(A=Add(A=@n85, B=@n86), B=NamedRerouteUsage())) {name=WPO, nodeColor=<rgba 0 1 0.1797 1>}
  n87 = SamplePhysicsVectorField!(World Position=NamedRerouteUsage()) {fieldTarget=Vector_LinearVelocity}
  n88 = MaterialFunctionCall(Vector 3=Add(A=@n87) {constB=0.0001}) {materialFunction=/Engine/Functions/Engine_MaterialFunctions02/Utility/VectorLength}
  n89 = Constant()
  NamedRerouteDeclaration(None=ShadingPathSwitch(Default=Saturate(None=@n88.V3 Length), Mobile=@n89) {desc=Migrated from FeatureLevelSwitch}) {name=PhysicsFieldMask, nodeColor=<rgba 0 1 0.3906 1>}
  n90 = Constant3Vector() {constant=<rgba 0 0 1 0>}
  NamedRerouteDeclaration(None=ShadingPathSwitch(
    Default=MaterialFunctionCall(Vector=@n87, Default=@n90) {materialFunction=&SafeNormalize}.Result
    Mobile=Constant3Vector() {constant=<rgba 0 0 1 0>}
   ) {desc=Migrated from FeatureLevelSwitch}) {name=PhysicsFieldDirection, nodeColor=<rgba 0.3828 0 1 1>}
  n91 = Min(A=@n88.V3 Length, B=ScalarParameter() {defaultValue=1, group=PhysicsField, parameterName=MaxRotation_PhysicsField_Clamp})
  n92 = Time()
  n93 = VertexColor()
  n94 = Add(A=@n93.R, B=@n93.R)
  n95 = Multiply(
   A=ScalarParameter() {defaultValue=351, group=PhysicsField, parameterName=PhysicsField_NoiseOffset}
   B=Multiply(A=Add(A=@n94, B=@n93.B)) {constB=33.33}
  )
  n96 = Frac(None=Multiply(A=ScalarParameter() {defaultValue=2.5, group=PhysicsField, parameterName=MaxRotation_PhysicsField_NoiseSpeed}, B=Add(A=@n92, B=@n95)))
  n97 = OneMinus(None=@n96)
  n98 = LinearInterpolate(
   A=Multiply(A=Multiply(A=@n96, B=@n97)) {constB=2}
   Alpha=ScalarParameter() {defaultValue=0.6, group=PhysicsField, parameterName=MaxRotation_PhysicsField_MInBend}
  )
  NamedRerouteDeclaration(None=ShadingPathSwitch(Default=Multiply(A=@n91, B=@n98), Mobile=@n89) {desc=Migrated from FeatureLevelSwitch}) {name=PhysicsFieldStrength, nodeColor=<rgba 1 0 0.2109 1>}
  n99 = MaterialFunctionCall() {materialFunction=&DebugTimeSine}
  n100 = TextureObjectParameter() {parameterName=InnerGlowNoise, texture=/Game/Global/Textures/Default/Tileable_Patterns/T_Blowing_Smoke_Pattern2_Tile_DISP}
  n101 = MaterialFunctionCall() {materialFunction=&LocalPosition}
  n102 = MaterialFunctionCall(Float3=@n101.Local Position) {materialFunction=&BreakOutFloat3Components}
  n103 = Abs(None=Transform(None=Constant3Vector() {constant=<rgba 1 1 1 0>}) {transformSourceType=TRANSFORMSOURCE_Local, transformType=TRANSFORM_Camera})
  n104 = Multiply(
   A=Time()
   B=Add(A=ScalarParameter() {defaultValue=5, parameterName=InnerGlowNoisePanningSpeed}, B=Multiply(A=@n103) {constB=-0.0001}) {constA=-1}
  ) {constB=-1}
  n105 = Multiply(B=@n104) {constA=0.1}
  n106 = Multiply(
   A=MaterialFunctionCall(X=@n102.R, Y=Add(A=@n102.G, B=@n105), Z=Add(A=@n101.Local Position, B=@n104)) {materialFunction=&MakeFloat3}.Result
   B=ScalarParameter() {defaultValue=1, parameterName=InnerGlowNoiseScale}
  ) {constB=2}
  n107 = Constant() {r=200}
  n108 = Multiply(A=Multiply(A=MaterialFunctionCall!(
     ColorTex=@n100
     NormalTex
     Pos=@n106
     Nrm
     ColorUVScale=@n107
     ColorBlendFactor
     NormalStrength
     NormalUVScale
     NormalBlendFactor
     Local Space Inputs
    ) {materialFunction=/Game/Global/MaterialFunctions/MF_BiPlanar}.Color)) {constB=0.05}
  n109 = Fresnel(
   ExponentIn=Subtract(B=Multiply(A=@n99.Result) {constB=0.1})
   Normal=Add(A=Transform(None=@n108) {transformSourceType=TRANSFORMSOURCE_Local, transformType=TRANSFORM_Tangent}, B=PixelNormalWS())
  ) {baseReflectFraction=0.01}
  n110 = NamedRerouteDeclaration(None=Saturate(None=LinearInterpolate(Alpha=@n109) {constA=1, constB=-0.5})) {name=Fresnel, nodeColor=<rgba 1 0 0.3516 1>}
  n111 = TransformPosition(None=VectorParameter() {parameterName=InnerGlowLocation, defaultValue=<rgba 0 0 40 1>}.RGB) {transformType=TRANSFORMPOSSOURCE_World}
  n112 = Constant()
  n113 = Multiply(A=TextureCoordinate()) {constB=2}
  n114 = Multiply(
   A=MaterialFunctionCall(
    X=@n112
    Y=@n112
    Z=TextureSample(UVs=@n113) {samplerType=SAMPLERTYPE_Masks, texture=/Game/Weapons/WeaponSkins/Wraps/Textures/Mash/Mash_Noise}.R
   ) {materialFunction=&MakeFloat3}.Result
   B=Constant() {r=5}
  ) {constB=5}
  n115 = ViewSize()
  n116 = Multiply(
   A=MaterialFunctionCall!(Center Point World Position=Add(A=@n111, B=@n114)) {materialFunction=/Engine/Functions/Engine_MaterialFunctions02/UVs/WorldSpaceAlignedScreenCoordinates}.Offset and Scaled UVS
   B=Divide(A=ComponentMask(None=@n115) {g=true}, B=Constant() {r=1080})
  )
  n117 = VectorParameter() {parameterName=InnerGlowTransform, defaultValue=<rgba 0.9 1.5 1 1>}
  n118 = SphereMask(
   A=Multiply(A=Divide(A=@n116) {constB=0.5}, B=ComponentMask(None=@n117.RGB) {g=true, r=true})
   B=Constant()
   Radius=Multiply(
    A=ScalarParameter() {defaultValue=0.07, parameterName=InnerGlowScale}
    B=MaterialFunctionCall() {materialFunction=&ObjectScale}.Scale X
   ) {constA=0.05}
  ) {attenuationRadius=0.05, hardnessPercent=0}
  n119 = ScalarParameter() {defaultValue=2, parameterName=InnerGlowEmissive_Night}
  n120 = ScalarParameter() {defaultValue=5, parameterName=InnerGlowEmissive}
  n121 = ScalarParameter() {defaultValue=2, parameterName=InnerGlowEmissive_Evening}
  n122 = Multiply(
   A=Saturate(None=Power(Base=@n118) {constExponent=1})
   B=Multiply(
    A=VectorParameter() {parameterName=InnerGlowColor, defaultValue=<rgba 1 0.4536 0.125 1>}.RGB
    B=MaterialFunctionCall(Value_Morning=@n119, Value_Day=@n120, Value_Evening=@n121, Value_Night=@n119) {materialFunction=&Scalar_Time_of_Day}.Result
   )
  )
  n123 = Add(A=Multiply(A=Constant() {r=-0.1}, B=Power(Base=@n110) {constExponent=1}) {constB=0}, B=@n122)
  n124 = MaterialFunctionCall(In=OneMinus(None=Fresnel() {exponent=0.3}), Contrast=Constant() {r=0.2}) {materialFunction=&CheapContrast}
  n125 = TextureSample(UVs=Multiply(A=TextureCoordinate())) {samplerType=SAMPLERTYPE_LinearColor, texture=/Game/Packages/DS_Fortnite_Terrain_NoLOD/Textures/Cloud_Height5}
  n126 = Constant()
  n127 = Saturate(None=Add(A=MaterialFunctionCall(In=@n125.R, Contrast=@n126) {materialFunction=&CheapContrast}.Result) {constB=0.3})
  n128 = Multiply(A=Add(A=Saturate(None=@n123), B=NamedRerouteUsage()), B=Multiply(A=@n122, B=Multiply(A=@n124.Result, B=@n127)))
  n129 = NamedRerouteUsage()
  NamedRerouteDeclaration(None=Multiply(A=Add(A=@n128, B=@n129), B=NamedRerouteUsage())) {name=Emissive, nodeColor=<rgba 0 0.0625 1 1>}
  n130 = TextureSampleParameter2D() {
   parameterName=FX_Mask
   samplerType=SAMPLERTYPE_Masks
   texture=/Game/Environments/Asteria/Foliage/Plants/BouncePadFlower/Textures/T_Foliage_BouncePadFlower_A_A_FX
  }
  NamedRerouteDeclaration(None=@n130.R) {name=YolkArea, nodeColor=<rgba 0 1 0.9297 1>}
  n131 = TextureSampleParameter2D() {
   parameterName=M
   samplerType=SAMPLERTYPE_Masks
   texture=&T_Foliage_BouncePadFlower_A_A_M
  }
  NamedRerouteDeclaration(None=@n131.R) {name=SSMask, nodeColor=<rgba 0 0.7422 1 1>}
  n132 = TextureCoordinate()
  n133 = TextureObject() {texture=/Game/Characters/Player/Male/Medium/Bodies/M_MED_HighTower_Tapas/Textures/T_Hightower_Tapas_Star}
  n134 = LinearInterpolate(
   A=TextureSample(UVs=Multiply(A=@n132) {constB=3}, Tex=@n133) {samplerSource=SSM_Wrap_WorldGroupSettings}.R
   B=TextureSample(UVs=Multiply(A=@n132) {constB=1.5}, Tex=@n133) {samplerSource=SSM_Wrap_WorldGroupSettings}.R
   Alpha=MaterialFunctionCall(Speed=Constant() {r=1}) {materialFunction=&DebugTimeSine}.Result
  )
  n135 = TextureObject() {samplerType=SAMPLERTYPE_Grayscale, texture=/Game/Weapons/FORT_Melee/Pickaxe_Cube_Ninja_Male/FX/T_CellNoise}
  n136 = Add(
   A=MaterialFunctionCall() {materialFunction=&LocalPosition}.Local Position
   B=Multiply(A=Time())
  )
  n137 = Constant() {r=75}
  n138 = Constant() {r=1}
  n139 = Constant() {r=500}
  n140 = ComponentMask(None=Saturate(None=MaterialFunctionCall(ColorTex=@n135, Pos=@n136, ColorUVScale=@n137, NormalStrength=@n138, NormalBlendFactor=@n139) {materialFunction=&MF_BiPlanar}.Color)) {r=true}
  n141 = ScalarParameter() {defaultValue=1, parameterName=StarEmissive_Night}
  n142 = ScalarParameter() {defaultValue=10, parameterName=StarEmissive}
  NamedRerouteDeclaration(None=Multiply(
    A=Multiply(A=@n134, B=@n140)
    B=MaterialFunctionCall(Value_Morning=@n141, Value_Day=@n142, Value_Evening=@n141, Value_Night=@n141) {materialFunction=&Scalar_Time_of_Day}.Result
   ) {constB=5}) {name=Stars, nodeColor=<rgba 0.5937 1 0 1>}
  NamedRerouteDeclaration(None=@n43.Base Color) {name=BaseColorIN, nodeColor=<rgba 0.7578 0 1 1>}
  OneMinus(None=@n109)
  Panner(Coordinate=@n113, Time=Time()) {speedX=-0.02, speedY=-0.05}
  Reroute(None=@n70.StormMask)
  Reroute(None=@n31)
  Reroute(None=@n75)
  Reroute(None=@n83)
  TextureSample() {samplerType=SAMPLERTYPE_Grayscale, texture=&T_CellNoise}
  VectorParameter() {parameterName=InnerFresnelColor, defaultValue=<rgba 0 0 0 1>}
  WorldPosition()
  WorldPosition()
 ]
}
